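# Animation Blueprint for the VealCadet pickaxe that makes its eye glance around at random.
# Each update interpolates EyeRotation yaw and roll toward target angles with FInterpTo, keeping pitch at zero.
# Once yaw is within a degree of its target, a pause timer counts down, then picks a random pause, yaw, roll and speed.
# The AnimGraph applies EyeRotation additively in bone space to MSTR_eye over the input pose.
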AnimBlueprint /BRCosmetics/Weapons/FORT_Melee/Pickaxe_VealCadet/Meshes/VealCadet_Axe_AnimBP
target_skeleton=&/VealCadet_Axe_Skeleton
graphs [
 AnimGraph {nodes [
   AnimGraphNode_SubInput_0 = Input Pose()
   n1 = Local To Component(LocalPose=@AnimGraphNode_SubInput_0)
   n2 = Get EyeRotation() -> (EyeRotation:Rotator)
   Output Pose(Result=Component To Local(ComponentPose="Transform (Modify) Bone - Bone: MSTR_eye"(ComponentPose=@n1, Alpha=1, Translation="0,0,0", Rotation=@n2, Scale="1,1,1") {rotationMode=BMM_Additive, rotationSpace=BCS_BoneSpace, boneToModify {boneName=MSTR_eye}})) {layerGroup=None}
  ]}
 EventGraph {nodes [
   n1 = Event BlueprintUpdateAnimation()
   > Set Delta Time X(Delta Time X=@n1.DeltaTimeX)
   > Set!(
    execute
    EyeRotation_Roll:float=FInterpTo(
     Current=Get!() -> (EyeRotation_Roll:float, EyeRotation_Pitch:float, EyeRotation_Yaw:float)
     Target=Get TargetAngleRoll() -> (TargetAngleRoll:double)
     DeltaTime=Get Delta Time X()
     InterpSpeed=Get Speed() -> (Speed:double)
    )
    EyeRotation_Pitch:float=0
    EyeRotation_Yaw:float=FInterpTo(Current=Get().EyeRotation_Yaw, Target=Get TargetAngleYaw() -> (TargetAngleYaw:double), DeltaTime=Get Delta Time X(), InterpSpeed=Get Speed())
   ) -> (then, Output_Get:Rotator)
   n2 = float - float(A=Get().EyeRotation_Yaw, B=Get TargetAngleYaw())
   > n3 = Set bIsPaused(bIsPaused:bool="float < float"(A="Absolute (Float)"(A=@n2), B=1)) -> (Output_Get:bool)
   > Branch(Condition=@n3)
   > n4 = Set PauseTimer(PauseTimer:double=float - float(A=Get PauseTimer() -> (PauseTimer:double), B=Get Delta Time X())) -> (Output_Get:double)
   > Branch(Condition="float <= float"(A=@n4) -> (ReturnValue=~))
   > Set PauseTimer(PauseTimer=RandomFloatInRange(Min=1, Max=3))
   > Set TargetAngleYaw(TargetAngleYaw:double=RandomFloatInRange(Min=-26, Max=25)) -> (Output_Get:double)
   > Set TargetAngleRoll(TargetAngleRoll:double=RandomFloatInRange(Min=-12, Max=10)) -> (Output_Get:double)
   > Set Speed(Speed:double=RandomFloatInRange(Min=2, Max=4)) -> (Output_Get:double)
   TryGetPawnOwner()
  ]}
]
variables [
 bIsPaused {type=boolean}
 Delta Time X {type=number}
 EyeRotation {type=Rotator}
 PauseTimer {type=number}
 Speed {type=number}
 TargetAngleRoll {type=number}
 TargetAngleYaw {type=number}
]
